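; Blueprint data object holding pending page-load state for the tournament leaderboard, storing a list offset and a page.
; A list offset of -1 marks the info as empty, and IsValidInfo returns true for any other offset.
; Reset sets the offset to that sentinel and copies the same value into the page.
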(fn IsValidInfo
  (return (!= (Variables|Default|GetListOffset) -1)))

(fn Reset
  (bind _output_get (Variables|Default|SetListOffset -1))
  (Variables|Default|SetPage _output_get))
